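# Sound cue for the SkeletonHunter cosmetic pickaxe's impact sound.
# A Random node picks among six equally weighted impact waves, feeding a slight volume Modulator and a short fixed Delay at the root.
# It primes on load, lowers pitch and volume, and uses the shared pickaxe impact attenuation, impacts sound class and property sheet.
# Overridden concurrency caps simultaneous plays per owner, stopping the oldest.
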SoundCue /BRCosmetics/Sounds/PickAxes/SkeletonHunter/SkeletonHunter_Impact_Cue
root=Delay_1
nodes [
 {id=Delay_1, children [Modulator_0], properties {delayMax=0.1, delayMin=0.1}}
 {id=Modulator_0, children [Random_2], properties {volumeMax=1.1, volumeMin=1}}
 {
  id=Random_2
  children [WavePlayer_8, WavePlayer_9, WavePlayer_10, WavePlayer_11, WavePlayer_12, WavePlayer_13]
  properties {weights [1,1,1,1,1,1]}
 }
 {id=WavePlayer_10, properties {soundWaveAssetPtr=&/SkeletonHunter_Impact_03}}
 {id=WavePlayer_11, properties {soundWaveAssetPtr=&/SkeletonHunter_Impact_04}}
 {id=WavePlayer_12, properties {soundWaveAssetPtr=&/SkeletonHunter_Impact_05}}
 {id=WavePlayer_13, properties {soundWaveAssetPtr=&/SkeletonHunter_Impact_06}}
 {id=WavePlayer_8, properties {soundWaveAssetPtr=&/SkeletonHunter_Impact_01}}
 {id=WavePlayer_9, properties {soundWaveAssetPtr=&/SkeletonHunter_Impact_02}}
]
properties {
 attenuationSettings=/Game/Sounds/Attenuation/Fort_Pickaxe_Impact_Att
 audioPropertiesSheet=/BRCosmetics/Sounds/Settings/PropertySheets/Pickaxe/APS_Cosmetic_Pickaxe_Impact
 bOverrideConcurrency=true
 bPrimeOnLoad=true
 maxDistance=4300
 pitchMultiplier=0.9
 soundClassObject=&Fort_Impacts
 volumeMultiplier=0.6525
 attenuationOverrides {
  bAttenuateWithLPF=true
  distanceAlgorithm=Logarithmic
  falloffDistance=800
  lPFRadiusMax=800
  lPFRadiusMin=400
  attenuationShapeExtents {x=300}
 }
 concurrencyOverrides {bLimitToOwner=true, maxCount=3, resolutionRule=StopOldest}
 concurrencySet [/Game/Sounds/Concurrency_Rules/OneMaximum_Owner_StopOldest]
}
